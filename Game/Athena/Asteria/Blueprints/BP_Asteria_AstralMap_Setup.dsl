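; Astral Map setup actor for Asteria that drives staged rift gate visuals and data layer state.
; On authority, calendar events SC01/SC02 (or a debug stage in PIE) set the overload amount and activate the SC_01 data layer, otherwise unloading it.
; Dormancy is flushed before replication; OnRep remaps overload through a curve into the telescope material parameter collection.
; A separate SkyTear OnRep, unset here, drives rift gate material progress and audio.
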(event EventBeginPlay
  (Networking|SwitchHasAuthority
    (:Authority
      (Utilities|PieDebug
        (:DebugPie
          (switch Utilities|FlowControl|Switch|SwitchonEnum_Asteria_RIft (Variables|Default|GetStages)
            (:NewEnumerator0
              ;[1 1 1] Debug
              ;;[1 1 1] Rift Gate Debug Progression
              (CallFunction|SetOverloadAmount (Variables|Default|GetDebug%20Time%20Span))
              (bind _returnvalue (WorldPartition|GetDataLayerManager))
              (Utilities|IsValid _returnvalue
                (:Is%20Valid
                  (DataLayers|SetDataLayerRuntimeState _returnvalue "/Game/Athena/Asteria/DataLayers/Asteria_DL_SC_01.Asteria_DL_SC_01" "Activated" false :!class /Script/Engine.DataLayerManager))))
            (:NewEnumerator1
              (CallFunction|SetOverloadAmount 1.0)
              (bind _returnvalue_2 (WorldPartition|GetDataLayerManager))
              (Utilities|IsValid _returnvalue_2
                (:Is%20Valid
                  (DataLayers|SetDataLayerRuntimeState _returnvalue_2 "/Game/Athena/Asteria/DataLayers/Asteria_DL_SC_01.Asteria_DL_SC_01" "Activated" false :!class /Script/Engine.DataLayerManager))))
            (:NewEnumerator2
              (bind _returnvalue_3 (WorldPartition|GetDataLayerManager))
              (DataLayers|SetDataLayerRuntimeState _returnvalue_3 "/Game/Athena/Asteria/DataLayers/Asteria_DL_SC_01.Asteria_DL_SC_01" "Unloaded" false :!class /Script/Engine.DataLayerManager))))
        (:NotDebugPie
          (Class|FortAsyncActionCurrentPlaylistReady|PlaylistDataReadyAsync
            (:OnReady
              ;[1 1 1] Rift Gate Overload Stage 1/2
              (bind (_biseventactive _timesincebegin _timeuntilend _timespanratio) (CalendarEvents|GetCalendarEventInformation "SC01"))
              (if _biseventactive
                (CallFunction|SetOverloadAmount _timespanratio)
                (bind _returnvalue_4 (WorldPartition|GetDataLayerManager))
                (DataLayers|SetDataLayerRuntimeState _returnvalue_4 "/Game/Athena/Asteria/DataLayers/Asteria_DL_SC_01.Asteria_DL_SC_01" "Activated" false :!class /Script/Engine.DataLayerManager)
                (else
                  (bind _biseventactive_2 (CalendarEvents|GetCalendarEventInformation "SC02"))
                  (if _biseventactive_2
                    (CallFunction|SetOverloadAmount 1.0)
                    (bind _returnvalue_5 (WorldPartition|GetDataLayerManager))
                    (DataLayers|SetDataLayerRuntimeState _returnvalue_5 "/Game/Athena/Asteria/DataLayers/Asteria_DL_SC_01.Asteria_DL_SC_01" "Activated" false :!class /Script/Engine.DataLayerManager)
                    (else
                      ;;[1 1 1] Failsafe if there is no Calendar event to read, Set Default Rift Gate State
                      (bind _returnvalue_6 (WorldPartition|GetDataLayerManager))
                      (Utilities|IsValid _returnvalue_6
                        (:Is%20Valid
                          (DataLayers|SetDataLayerRuntimeState _returnvalue_6 "/Game/Athena/Asteria/DataLayers/Asteria_DL_SC_01.Asteria_DL_SC_01" "Unloaded" false :!class /Script/Engine.DataLayerManager))))))))))))))

(fn ModifyOverloadValue
  (params (inOverload :type Float%20%28double-precision%29))
  ;[0.3435 1 0] ensures all visuals hit 1 of the 7 stages. Start on stage 1 (now 0.14..., was previously 0.0)
  (return (Math|Curves|GetFloatValue (Variables|Default|GetTimingMappings) inOverload)))

(fn OnRep_OverloadAmount
  (CallFunction|SetOverload (Variables|Default|GetOverload%20Amount)))

(fn OnRep_SkyTearAmount
  (Default|SetSkyTear (Variables|Default|GetSkyTear%20Amount)))

(fn SetOverloadAmount
  (params (Overload%20Amount :type Float%20%28double-precision%29))
  (Networking|FlushNetDormancy)
  (Variables|Default|SetOverload%20Amount Overload%20Amount))

(fn SetOverload
  (params (Overload :type Float%20%28double-precision%29))
  (Rendering|Material|SetScalarParameterValue "/Game/Environments/Asteria/Props/Ancients/Ancients_Telescope_A/FX/MPC_Ch4s3_Telescope.MPC_Ch4s3_Telescope" "Alpha" (CallFunction|ModifyOverloadValue Overload) :!class /Script/Engine.KismetMaterialLibrary))

(fn SetSkyTear
  (params (SkyTear :type Float%20%28double-precision%29))
  ;[1 1 1] Change Overload and Audio Param to SkyTear
  (Rendering|Material|SetScalarParameterValue "/Game/Athena/Items/ForagedItems/Rift/FX/RiftGate/MPC_Riftgate.MPC_Riftgate" "TearProgress" SkyTear :!class /Script/Engine.KismetMaterialLibrary)
  (Default|CallSkyTearAudioParamED SkyTear))

(fn ConstructionScript)
