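; Asteria lobby island setup actor that, on the server, unloads lobby island data layers once the match moves past setup and warmup.
; Setup checks the current phase, otherwise listens for game phase update events until unloading is due.
; HideLayers unloads island, landscape and an extra layer, but activates the lobby navmesh layer.
; A None phase unloads only if player 0 is far away; editor runs retry setup until that pawn exists.
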(event Custom|TrySetup
  (Utilities|IsValid (Game|GetPlayerPawn 0)
    (:Is%20Valid
      (CallFunction|Setup))
    (:Is%20Not%20Valid
      (Utilities|Time|SetTimerbyEvent (EventDispatchers|CreateEvent :!event Custom|TrySetup) 0.1))))

(event EventBeginPlay
  (Networking|SwitchHasAuthority
    (:Authority
      (if (Editor|InEditororPIE)
        (CallFunction|TrySetup) ; In PIE, when playing "from here" the pawn may not be ready so we check for it first
        (else
          (CallFunction|Setup))))))

(event EventEndPlay
  (params (EndPlayReason :type EEndPlayReason%20Enum :default "Destroyed"))
  (bind _gamephaseupdatedasynctask (Variables|Default|GetGamePhaseUpdatedAsyncTask)
    (:then
      (GameplayEventRouter|StopListeningtoStatefulEvent _gamephaseupdatedasynctask))
    (:else)))

(event Custom|Setup
  (bind _returnvalue (BattleRoyaleGamePhaseLogic|GetBattleRoyaleGamePhaseLogicComponent))
  (bind _returnvalue_2 (BattleRoyaleGamePhaseLogic|GetGamePhase _returnvalue))
  (bind _returnvalue_3 (Utilities|String|EnumtoString _returnvalue_2))
  (bind _returnvalue_4 (Utilities|String|Append "Lobby Island Unloader: Phase at start of game is " _returnvalue_3 ".  Unloading Lobby Data Layer!"))
  ;[1 1 1] Initial check for late players - if we are in any stage that isnt setup or warmup, unload lobby island.
  (if (CallFunction|IsPhaseforUnloading _returnvalue_2)
    (CallFunction|HideLayers)
    (Fortnite|FortShippingLog self _returnvalue_4)
    (Development|PrintString _returnvalue_4)
    (else
      (bind _returnvalue_5 (GameplayEventRouter|GetEventRouter "/Script/FortniteGame.FortGameState" self))
      (bind (_asynctaskproxy _payload) (GameplayEventRouter|StartListeningToStatefulEvent _returnvalue_5 "/Script/FortniteGame.FortGamePhaseUpdatedEvent")
        (:then
          (Variables|Default|SetGamePhaseUpdatedAsyncTask _asynctaskproxy))
        (:OnEventReceived
          _)
        (:OnSavedEventStateExists
          _))
      (CallFunction|HandleGamePhaseChanged (Utilities|Struct|BreakFortGamePhaseUpdatedEvent _payload)))))

(fn handleGamePhaseChanged
  (params (game%20phase :type EAthenaGamePhase%20Enum))
  (bind _returnvalue (Utilities|String|EnumtoString game%20phase))
  (bind _returnvalue_2 (Utilities|String|Append "Lobby Island Unloader: Phase Changed to " _returnvalue ".  Unloading Lobby Data Layer!"))
  (if (CallFunction|IsPhaseforUnloading game%20phase)
    (CallFunction|HideLayers)
    (Fortnite|FortShippingLog self _returnvalue_2)
    (Development|PrintString _returnvalue_2)
    (bind _gamephaseupdatedasynctask (Variables|Default|GetGamePhaseUpdatedAsyncTask)
      (:then
        (GameplayEventRouter|StopListeningtoStatefulEvent _gamephaseupdatedasynctask))
      (:else))))

(fn HideLayers
  (bind _returnvalue (WorldPartition|GetDataLayerManager))
  (DataLayers|SetDataLayerInstanceRuntimeState _returnvalue (DataLayers|GetDataLayerInstancefromAsset _returnvalue "/Game/Athena/Asteria/DataLayers/Asteria_DL_LobbyIslandNavMesh.Asteria_DL_LobbyIslandNavMesh") "Activated" false)
  (DataLayers|SetDataLayerInstanceRuntimeState _returnvalue (DataLayers|GetDataLayerInstancefromAsset _returnvalue "/Game/Athena/Asteria/DataLayers/Asteria_DL_LobbyIsland.Asteria_DL_LobbyIsland") "Unloaded" false)
  (DataLayers|SetDataLayerInstanceRuntimeState _returnvalue (DataLayers|GetDataLayerInstancefromAsset _returnvalue "/Game/Athena/Asteria/DataLayers/Asteria_DL_LobbyIslandLandscape.Asteria_DL_LobbyIslandLandscape") "Unloaded" false)
  (DataLayers|SetDataLayerInstanceRuntimeState _returnvalue (DataLayers|GetDataLayerInstancefromAsset _returnvalue (Variables|Default|GetExtraDLToHide)) "Unloaded" false))

(fn isPhaseForUnloading
  (params (phase :type EAthenaGamePhase%20Enum))
  (bind _returnvalue (Game|GetPlayerPawn 0))
  (bind _returnvalue_2 (Transformation|GetActorLocation))
  (bind _returnvalue_3 (Transformation|GetActorLocation _returnvalue))
  (bind _returnvalue_4 (Math|Vector|Distance(Vector) _returnvalue_3 _returnvalue_2))
  (bind _returnvalue_5 (> _returnvalue_4 1e+05))
  (if (Utilities|Enum|Equal(Enum) phase "None")
    (Development|PrintString (Utilities|String|Append "Distance to Player from Lobby Island Setup: " (Utilities|String|ToString(Float) _returnvalue_4) " > 100000.0 = " (Utilities|String|ToString(Boolean) _returnvalue_5)))
    (return _returnvalue_5)
    (else
      (return (Utilities|Select :Index phase :None false :Setup false :Warmup false :Aircraft true :SafeZones true :EndGame true :Count true)))))

(fn ConstructionScript)
