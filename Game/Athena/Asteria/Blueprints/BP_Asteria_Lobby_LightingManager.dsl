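; Lobby lighting manager that swaps between spawn island lighting and the main island Time of Day Manager by game phase.
; On authority, it spawns a hidden lighting actor, then sets up once the TODM is ready, or in editor once a pawn exists.
; Lobby phases hide the TODM and listen for phase updates; later phases restore it and destroy the lobby actor.
; For phase None, player distance from this actor decides instead.
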(event EventBeginPlay
  (bind _returnvalue (Game|GetPlayerCameraManager 0))
  ;[0.1622 0.3542 0.1288 1 48] Begin Play
  (Networking|SwitchHasAuthority
    (:Authority
      (Utilities|FlowControl|Sequence
        (:then_0
          (bind _returnvalue_2 (Game|SpawnActorfromClass "/Game/TimeOfDay/TODM_Sequence/Blueprint/BP_SpawnIsland_LightingActor.BP_SpawnIsland_LightingActor_C" (Math|Transform|MakeTransform (Transformation|GetActorLocation self)) "Undefined" "OverrideRootScale" 0 0))
          (bind _output_get (Variables|Default|SetLobbyLighting _returnvalue_2))
          (Fortnite|FortShippingLog self (Utilities|String|Append "LobbyLightingDebug: " "Lighting Actor Is Valid: " (Utilities|String|ToString(Boolean) (Utilities|IsValid _output_get :!class /Script/Engine.KismetSystemLibrary))))
          (Rendering|SetActorHiddenInGame _output_get true))
        (:then_1
          (if (Editor|InEditororPIE)
            (CallFunction|TrySetupLighting)
            (else
              (Class|FortAsyncActionTimeOfDayManagerReady|GetTimeOfDayManagerReady
                (:OnReady
                  (CallFunction|SetupLighting))))))
        (:then_2
          (Utilities|IsValid _returnvalue
            (:Is%20Valid
              (if (and (Utilities|String|Contains (Utilities|GetDisplayName _returnvalue) "Spectator") (Editor|InEditororPIE))
                (CallFunction|SetUpMainIslandLighting)))))))))

(event EventEndPlay
  (params (EndPlayReason :type EEndPlayReason%20Enum :default "Destroyed"))
  ;[0.1622 0.3542 0.1288 1 48] Begin Play
  (bind _gamephaseupdatedasynctask (Variables|Default|GetGamePhaseUpdatedAsyncTask)
    (:then
      (GameplayEventRouter|StopListeningtoStatefulEvent _gamephaseupdatedasynctask))
    (:else)))

(event Custom|TrySetupLighting
  ;[0.2031 0.2031 0.2031 1 36] Try and Set Up Lighting (PIE ONLY)
  (Fortnite|FortShippingLog self (Utilities|String|Append "LobbyLightingDebug: " "Event: Try Setup Lighting Is Called"))
  (Utilities|IsValid (Game|GetPlayerPawn 0)
    (:Is%20Valid
      (Fortnite|FortShippingLog self (Utilities|String|Append "LobbyLightingDebug: " "PlayerPawnIsValid"))
      (CallFunction|SetupLighting))
    (:Is%20Not%20Valid
      (Utilities|Time|SetTimerbyEvent (EventDispatchers|CreateEvent :!event Custom|TrySetupLighting) 0.1))))

(event Custom|SetupLighting
  (bind _returnvalue (BattleRoyaleGamePhaseLogic|GetBattleRoyaleGamePhaseLogicComponent))
  (bind _returnvalue_2 (BattleRoyaleGamePhaseLogic|GetGamePhase _returnvalue))
  (bind _shouldunload (Default|IsPhaseforUnloading _returnvalue_2))
  (bind _returnvalue_3 (Utilities|String|ToString(Boolean) _shouldunload))
  (bind _returnvalue_4 (Utilities|String|Append "LobbyLightingDebug: " "IsPhaseForUnloading: " _returnvalue_3))
  (bind _self self)
  ;[0.151 0.151 0.151 1 36] Set Up Lighting Main Event
  (Fortnite|FortShippingLog self (Utilities|String|Append "LobbyLightingDebug: " "Event: Setup Lighting IsCalled"))
  (if _shouldunload
    (Fortnite|FortShippingLog _self _returnvalue_4)
    (CallFunction|SetUpMainIslandLighting)
    (else
      (Fortnite|FortShippingLog _self _returnvalue_4)
      (CallFunction|SetupLobbyLighting)
      (bind _returnvalue_5 (GameplayEventRouter|GetEventRouter "/Script/FortniteGame.FortGameState" self))
      (bind (_asynctaskproxy _payload) (GameplayEventRouter|StartListeningToStatefulEvent _returnvalue_5 "/Script/FortniteGame.FortGamePhaseUpdatedEvent")
        (:then
          (Variables|Default|SetGamePhaseUpdatedAsyncTask _asynctaskproxy))
        (:OnEventReceived
          _)
        (:OnSavedEventStateExists
          _))
      (CallFunction|GamePhaseChanged (Utilities|Struct|BreakFortGamePhaseUpdatedEvent _payload)))))

(event Custom|GamePhaseChanged
  (params (GamePhase :type EAthenaGamePhase%20Enum))
  (bind _shouldunload (Default|IsPhaseforUnloading GamePhase))
  ;[0.1567 0.1756 0.5365 1 36] Check For Game Phase Change
  (Fortnite|FortShippingLog self (Utilities|String|Append "LobbyLightingDebug: " "Event: GamePhaseChanged IsCalled "))
  (Fortnite|FortShippingLog 0 (Utilities|String|Append "LobbyLightingDebug: " "IsPhaseForUnloading: " (Utilities|String|ToString(Boolean) _shouldunload)))
  (if _shouldunload
    (CallFunction|SetUpMainIslandLighting)))

(event Collision|EventActorBeginOverlap
  (params (OtherActor :type Actor%20Object%20Reference)))

(event EventTick
  (params (DeltaSeconds :type Float%20%28single-precision%29)))

(event Custom|SetUpMainIslandLighting
  (bind _lobbylighting (Variables|Default|GetLobbyLighting))
  ;[0.2344 0.2344 0.2344 1 36] Set Up Main Island Lighting
  (Fortnite|FortShippingLog self (Utilities|String|Append "LobbyLightingDebug: " "Event: SetupMainIslandLighting IsCalled"))
  (bind _returnvalue (Utilities|TryGetTODManager)
    (:Valid
      (TimeOfDay|SetTimeOfDayManagerVisibility _returnvalue true)
      (Fortnite|FortShippingLog self (Utilities|String|Append "LobbyLightingDebug: " "Set TODM Visible On Server")))
    (:Invalid
      _))
  (if (Utilities|IsValid (Variables|Default|GetLobbyLighting) :!class /Script/Engine.KismetSystemLibrary)
    (Fortnite|FortShippingLog self (Utilities|String|Append "LobbyLightingDebug: " "LightingActorIsValid"))
    (Rendering|SetActorHiddenInGame _lobbylighting true)
    (Utilities|FlowControl|Sequence
      (:then_0
        (bind _asbp_spawn_island_lighting_actor (Utilities|Casting|CastToBP_SpawnIsland_LightingActor _lobbylighting))
        (Class|BPSpawnIslandLightingActor|DisableAll _asbp_spawn_island_lighting_actor))
      (:then_1
        (Fortnite|FortShippingLog self (Utilities|String|Append "LobbyLightingDebug: " "LightingActorIsHidden"))
        (bind _gamephaseupdatedasynctask (Variables|Default|GetGamePhaseUpdatedAsyncTask)
          (:then
            (GameplayEventRouter|StopListeningtoStatefulEvent _gamephaseupdatedasynctask))
          (:else
            _))
        (bind _returnvalue_2 (Utilities|TryGetTODManager)
          (:Valid
            (Fortnite|FortShippingLog self (Utilities|String|Append "LobbyLightingDebug: " "TODM Visibility: " (Utilities|String|ToString(Boolean) (TimeOfDay|IsVisible _returnvalue_2)))))))
      (:then_2
        (Actor|DestroyActor _lobbylighting)))))

(event Custom|SetupLobbyLighting
  ;[0.1823 0.1823 0.1823 1 36] Set Up Lobby Lighting
  (Fortnite|FortShippingLog self (Utilities|String|Append "LobbyLightingDebug: " "Event: SetupLobbyLighting IsCalled"))
  (bind _returnvalue (Utilities|TryGetTODManager)
    (:Valid
      (TimeOfDay|SetTimeOfDayManagerVisibility _returnvalue false)
      (Fortnite|FortShippingLog self (Utilities|String|Append "LobbyLightingDebug: " "Set TODM Invisible On Server")))
    (:Invalid
      _))
  (if (Utilities|IsValid (Variables|Default|GetLobbyLighting) :!class /Script/Engine.KismetSystemLibrary)
    (Fortnite|FortShippingLog self (Utilities|String|Append "LobbyLightingDebug: " "LightingActorIsValid"))
    (Rendering|SetActorHiddenInGame (Variables|Default|GetLobbyLighting) false)
    (bind _returnvalue_2 (Utilities|TryGetTODManager)
      (:Valid
        (Fortnite|FortShippingLog self (Utilities|String|Append "LobbyLightingDebug: " "TODM Visibility: " (Utilities|String|ToString(Boolean) (TimeOfDay|IsVisible _returnvalue_2))))))))

(fn getTimeOfDayManagerRef
  (Utilities|IsValid 0
    (:Is%20Not%20Valid
      (Utilities|IsValid (Utilities|ToObject(Interface) (TimeOfDay|GetGlobalTimeOfDayManager))))))

(fn isPhaseForUnloading
  (params (phase :type EAthenaGamePhase%20Enum))
  (bind _returnvalue (Game|GetPlayerPawn 0))
  (bind _returnvalue_2 (Transformation|GetActorLocation))
  (bind _returnvalue_3 (Transformation|GetActorLocation _returnvalue))
  (bind _returnvalue_4 (Math|Vector|Distance(Vector) _returnvalue_3 _returnvalue_2))
  (bind _returnvalue_5 (> _returnvalue_4 1e+05))
  (if (Utilities|Enum|Equal(Enum) phase "None")
    (Development|PrintString (Utilities|String|Append "Distance to Player from Lobby Island Setup: " (Utilities|String|ToString(Float) _returnvalue_4) " > 100000.0 = " (Utilities|String|ToString(Boolean) _returnvalue_5)))
    (return _returnvalue_5)
    (else
      (Fortnite|FortShippingLog self "111")
      (return (Utilities|Select :Index phase :None false :Setup false :Warmup false :Aircraft true :SafeZones true :EndGame true :Count true)))))

(fn ConstructionScript)
