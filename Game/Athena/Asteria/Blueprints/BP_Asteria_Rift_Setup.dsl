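; Setup Blueprint that sets the Asteria Rift Gate overload and sky tear state at BeginPlay.
; The server reads RG/ST calendar events once the playlist is ready, or a debug stage enum in PIE, to set progress amounts and activate Rift Gate data layers.
; RepNotifies on both amounts push values to the MPC_Riftgate material parameter collection and call audio parameter dispatchers.
; A SetST event allows console testing of sky tear progress.
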(event Custom|SetST
  (params (SkyTear :type Float%20%28double-precision%29))
  ;[1 1 1] Debug
  ;;[1 1 1] Console Commands for PIE Testing
  (Variables|Default|SetSkyTear%20Amount SkyTear))

(event EventBeginPlay
  (Networking|SwitchHasAuthority
    (:Authority
      (Utilities|PieDebug
        (:DebugPie
          (switch Utilities|FlowControl|Switch|SwitchonEnum_Asteria_RIft (Variables|Default|GetStages)
            (:NewEnumerator0
              ;[1 1 1] Debug
              ;;[1 1 1] Rift Gate Debug Progression
              (Variables|Default|SetOverload%20Amount (Variables|Default|GetDebug%20Time%20Span)))
            (:NewEnumerator1
              (Variables|Default|SetOverload%20Amount 1.0))
            (:NewEnumerator2
              ;;[1 1 1] Sky Tear Debug Progression
              (Variables|Default|SetSkyTear%20Amount (Variables|Default|GetDebug%20Time%20Span))
              (bind _returnvalue (WorldPartition|GetDataLayerManager))
              (DataLayers|SetDataLayerRuntimeState _returnvalue "/Game/Athena/Asteria/DataLayers/Asteria_DL_RG_02.Asteria_DL_RG_02" "Activated" false :!class /Script/Engine.DataLayerManager))
            (:NewEnumerator3
              (Variables|Default|SetSkyTear%20Amount 1.0)))
          (bind _returnvalue_2 (WorldPartition|GetDataLayerManager))
          (DataLayers|SetDataLayerRuntimeState _returnvalue_2 "/Game/Athena/Asteria/DataLayers/Asteria_DL_RG_01.Asteria_DL_RG_01" "Activated" false :!class /Script/Engine.DataLayerManager))
        (:NotDebugPie
          (Class|FortAsyncActionCurrentPlaylistReady|PlaylistDataReadyAsync
            (:OnReady
              ;[1 1 1] Rift Gate Overload Stage 1/2
              (bind (_biseventactive _timesincebegin _timeuntilend _timespanratio) (CalendarEvents|GetCalendarEventInformation "RG01"))
              (if _biseventactive
                (Variables|Default|SetOverload%20Amount _timespanratio)
                (else
                  (bind _biseventactive_2 (CalendarEvents|GetCalendarEventInformation "RG02"))
                  (if _biseventactive_2
                    (Variables|Default|SetOverload%20Amount 1.0)
                    (else
                      ;[1 1 1] Rift Gate Sky Tear Stage 1/2
                      (bind (_biseventactive_3 _timesincebegin_2 _timeuntilend_2 _timespanratio_2) (CalendarEvents|GetCalendarEventInformation "ST01"))
                      (if _biseventactive_3
                        (Variables|Default|SetSkyTear%20Amount _timespanratio_2)
                        (else
                          (bind _biseventactive_4 (CalendarEvents|GetCalendarEventInformation "ST02"))
                          (if _biseventactive_4
                            (Variables|Default|SetSkyTear%20Amount 1.0)
                            (else
                              ;[1 1 1] Failsafe if there is no Calendar event to read, Set Default Rift Gate State
                              (bind _returnvalue_3 (WorldPartition|GetDataLayerManager))
                              (DataLayers|SetDataLayerRuntimeState _returnvalue_3 "/Game/Athena/Asteria/DataLayers/Asteria_DL_RG_01.Asteria_DL_RG_01" "Activated" false :!class /Script/Engine.DataLayerManager)
                              (Variables|Default|SetOverload%20Amount 0.0)))))
                      (bind _returnvalue_4 (WorldPartition|GetDataLayerManager))
                      (DataLayers|SetDataLayerRuntimeState _returnvalue_4 "/Game/Athena/Asteria/DataLayers/Asteria_DL_RG_02.Asteria_DL_RG_02" "Activated" false :!class /Script/Engine.DataLayerManager)))))
              (bind _returnvalue_5 (WorldPartition|GetDataLayerManager))
              (DataLayers|SetDataLayerRuntimeState _returnvalue_5 "/Game/Athena/Asteria/DataLayers/Asteria_DL_RG_01.Asteria_DL_RG_01" "Activated" false :!class /Script/Engine.DataLayerManager))))))))

(fn OnRep_OverloadAmount
  (CallFunction|SetOverload (Variables|Default|GetOverload%20Amount)))

(fn OnRep_SkyTearAmount
  (CallFunction|SetSkyTear (Variables|Default|GetSkyTear%20Amount)))

(fn SetOverload
  (params (Overload :type Float%20%28double-precision%29))
  (Rendering|Material|SetScalarParameterValue "/Game/Athena/Items/ForagedItems/Rift/FX/RiftGate/MPC_Riftgate.MPC_Riftgate" "Overload" Overload :!class /Script/Engine.KismetMaterialLibrary)
  ;[1 1 1] Audio parameter
  (Default|CallOverloadAudioParamED Overload))

(fn SetSkyTear
  (params (SkyTear :type Float%20%28double-precision%29))
  ;[1 1 1] Change Overload and Audio Param to SkyTear
  (Rendering|Material|SetScalarParameterValue "/Game/Athena/Items/ForagedItems/Rift/FX/RiftGate/MPC_Riftgate.MPC_Riftgate" "TearProgress" SkyTear :!class /Script/Engine.KismetMaterialLibrary)
  (Default|CallSkyTearAudioParamED SkyTear))

(fn ConstructionScript)
