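; Server-side Asteria setup Blueprint that, on BeginPlay, activates one of the TM_01 to TM_03 World Partition data layers.
; With Debug set in editor or PIE, DebugStage picks the layer directly.
; Otherwise, only if Ignore in Creative is set, it waits for playlist data, skips Creative playlists, then checks matching calendar events after initial sync to choose the layer.
; With both paths off, nothing is activated.
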(event EventBeginPlay
  (Networking|SwitchHasAuthority
    (:Authority
      (if (and (Variables|Default|GetDebug) (Editor|InEditororPIE))
        ;[1 1 1] Debug for PIE
        (switch int (Variables|Default|GetDebugStage)
          (:0
            (bind _returnvalue (WorldPartition|GetDataLayerManager))
            (DataLayers|SetDataLayerInstanceRuntimeState _returnvalue (DataLayers|GetDataLayerInstancefromAsset _returnvalue "/Game/Athena/Asteria/DataLayers/Asteria_DL_TM_01.Asteria_DL_TM_01") "Activated" false))
          (:1
            (bind _returnvalue_2 (WorldPartition|GetDataLayerManager))
            (DataLayers|SetDataLayerInstanceRuntimeState _returnvalue_2 (DataLayers|GetDataLayerInstancefromAsset _returnvalue_2 "/Game/Athena/Asteria/DataLayers/Asteria_DL_TM_02.Asteria_DL_TM_02") "Activated" false))
          (:2
            (bind _returnvalue_3 (WorldPartition|GetDataLayerManager))
            (DataLayers|SetDataLayerInstanceRuntimeState _returnvalue_3 (DataLayers|GetDataLayerInstancefromAsset _returnvalue_3 "/Game/Athena/Asteria/DataLayers/Asteria_DL_TM_03.Asteria_DL_TM_03") "Activated" false)))
        (elif (Variables|Default|GetIgnore%20in%20Creative)
          ;[1 1 1] Handle Ignoring in Creative if needed - copied from BP_CalendarEventBase
          (bind _gamestate (Class|FortAsyncActionCurrentPlaylistReady|PlaylistDataReadyAsync)
            (:OnReady
              (if (not (GameplayTags|HasTag (Class|FortGameStateAthena|GetAthenaPlaylistContextTags _gamestate) "(TagName=\"Athena.Creative\")"))
                (MCPContext|InitialCalendarSyncCompleted
                  (:OnSyncComplete
                    (bind _biseventactive (CalendarEvents|GetCalendarEventInformation "TM_01"))
                    (if (not _biseventactive)
                      (bind _biseventactive_2 (CalendarEvents|GetCalendarEventInformation "TM_02"))
                      (if _biseventactive_2
                        (bind _returnvalue_4 (WorldPartition|GetDataLayerManager))
                        (DataLayers|SetDataLayerInstanceRuntimeState _returnvalue_4 (DataLayers|GetDataLayerInstancefromAsset _returnvalue_4 "/Game/Athena/Asteria/DataLayers/Asteria_DL_TM_02.Asteria_DL_TM_02") "Activated" false)
                        (else
                          (bind _biseventactive_3 (CalendarEvents|GetCalendarEventInformation "TM_03"))
                          (if _biseventactive_3
                            (bind _returnvalue_5 (WorldPartition|GetDataLayerManager))
                            (DataLayers|SetDataLayerInstanceRuntimeState _returnvalue_5 (DataLayers|GetDataLayerInstancefromAsset _returnvalue_5 "/Game/Athena/Asteria/DataLayers/Asteria_DL_TM_03.Asteria_DL_TM_03") "Activated" false)
                            (else
                              _)))))
                    (bind _returnvalue_6 (WorldPartition|GetDataLayerManager))
                    (DataLayers|SetDataLayerInstanceRuntimeState _returnvalue_6 (DataLayers|GetDataLayerInstancefromAsset _returnvalue_6 "/Game/Athena/Asteria/DataLayers/Asteria_DL_TM_01.Asteria_DL_TM_01") "Activated" false))))))
          (else
            _))))))

(fn ConstructionScript)
